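# Translucent, unlit, two-sided material for glider speed-line Niagara mesh particles.
# Emissive blends time-of-day color toward Color1 through turbulence noise, tinted by particle color.
# Opacity multiplies two panning noise layers by UV edge, view-angle and depth fades, and particle alpha.
# World position offset multiplies two switchable terms: curve-atlas-shaped SplineThicken width and a clamped offset from particle position, both driven by dynamic parameters.
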Material /Game/Gadgets/Assets/VinderTech_GliderChute/DefaultFX/M_Glider_SpeedLines
graph {
 properties {
  bUsedWithNiagaraMeshParticles=true
  blendMode=BLEND_Translucent
  pixelDepthOffsetMode=PDOM_Legacy
  shadingModel=MSM_Unlit
  twoSided=true
  usedShadingModels=Unlit
 }
 nodes [
  Constant3Vector() {constant=<rgba 1 1 1 0>}
  n1 = MaterialFunctionCall(
   Color_Morning=VectorParameter() {parameterName=MorningShadowColor, defaultValue=<rgba 0.5776 0.1779 0.0976 1>}.RGB
   Color_Day=VectorParameter() {parameterName=DayShadowColor, defaultValue=<rgba 0.7011 0.521 0.2789 1>}.RGB
   Color_Evening=VectorParameter() {parameterName=EveningShadowColor, defaultValue=<rgba 0.8879 0.484 0.3515 1>}.RGB
   Color_Night=VectorParameter() {parameterName=NightShadowColor, defaultValue=<rgba 0.1141 0.1494 0.3712 1>}.RGB
  ) {materialFunction=&Color_Time_of_Day}
  n2 = VectorParameter() {parameterName=Color1, defaultValue=<rgba 6.96 7.4808 8 0>}
  n3 = ScalarParameter() {defaultValue=1, parameterName=UMultiplier1}
  n4 = ScalarParameter() {defaultValue=1, parameterName=VMultiplier1}
  n5 = TextureSample(UVs=Multiply(A=TextureCoordinate() {uTiling=0.4, vTiling=2}, B=AppendVector(A=@n3, B=@n4))) {samplerType=SAMPLERTYPE_LinearColor, texture=/Game/Effects/Fort_Effects/Effects/Environment/Desert/Dust/Organize/T_Turbulence_Seamless}
  n6 = ParticleColor()
  MaterialFunctionCall(Emissive=Multiply(A=LinearInterpolate(A=@n1.Result, B=@n2.RGB, Alpha=@n5.G), B=@n6.RGB)) {materialFunction=&MF_FX_EmissiveAdjustments}
  n7 = DynamicParameter() {defaultValue=<rgba 0 1 1 1>, paramNames [Random Offset, Spline Width Mult, SplieWidthBase, SplineWidthTip]}
  n8 = ScalarParameter() {defaultValue=1, group=SplineThicken, parameterName=Spline Width Mult}
  n9 = Multiply(
   A=Multiply(
    A=MaterialFunctionCall(WidthBase=@n7.SplieWidthBase, WidthTip=@n7.SplieWidthBase) {materialFunction=&SplineThicken}.WorldPositionOffset
    B=Multiply(A=@n8, B=@n7.Spline Width Mult)
   )
   B=ScalarParameter() {defaultValue=1, group=SplineThicken, parameterName=Spline Width Mult 02}
  )
  n10 = ComponentMask(None=TextureCoordinate()) {g=true}
  n11 = CurveAtlasRowParameter(CurveTime=@n10) {
   atlas=/Game/Athena/LinearColorCurves/A_Atlas_Turtleneck_Glider
   curve=/Game/Accessories/FORT_Backpacks/Backpack_M_MED_Calavera/FX/Curve_Backpack_Calavera_01
   defaultValue=39
   group=SplineThicken
   parameterName=Curve Shape
  }
  n12 = CurveAtlasRowParameter(CurveTime=@n10) {
   atlas=&A_Atlas_Turtleneck_Glider
   curve=&Curve_Backpack_Calavera_01
   defaultValue=39
   group=SplineThicken
   parameterName=Curve Shape
  }
  n13 = LinearInterpolate(A=@n11.A, B=@n12.B, Alpha=ScalarParameter() {group=SplineThicken, parameterName=Lerp Shape Type})
  n14 = DynamicParameter() {
   desc=Parameters set by Niagara
   parameterIndex=3
   defaultValue=<rgba 0 1 1 1>
   paramNames [ShapeTypeShift, Spline Width Mult, Dissolve, Dissolve Edge Intensity]
  }
  n15 = MaterialFunctionCall(Float3=Subtract(
    A=MaterialFunctionCall(X=@n14.ShapeTypeShift, Y=@n14.ShapeTypeShift, Z=@n14.ShapeTypeShift) {materialFunction=&MakeFloat3}.Result
    B=ParticlePositionWS()
   )) {materialFunction=&BreakOutFloat3Components}
  n16 = ScalarParameter() {defaultValue=100, group=Interpolate, parameterName=Interpolate Clamp XY}
  n17 = ScalarParameter() {defaultValue=20, group=Interpolate, parameterName=Interpolate Clamp Z}
  n18 = AppendVector(
   A=Clamp(None=AppendVector(A=@n15.R, B=@n15.R), Min=Multiply(A=@n16) {constB=-1}, Max=@n16) {maxDefault=1000, minDefault=-1000}
   B=Clamp(None=@n15.B, Min=Multiply(A=@n17) {constB=-1}, Max=@n17) {minDefault=-1}
  )
  n19 = TextureCoordinate()
  n20 = ComponentMask(None=OneMinus(None=@n19)) {g=true}
  n21 = OneMinus(None=@n19)
  n22 = Multiply(A=Multiply(A=ComponentMask(None=@n21) {g=true}, B=ComponentMask(None=@n19) {g=true})) {constB=1.5}
  n23 = DynamicParameter() {parameterIndex=1, defaultValue=<rgba 0 10 1 1>, paramNames [Random Offset, Noise Intensty, Dissolve, Dissolve Edge Intensity]}
  n24 = LinearInterpolate(
   A=ScalarParameter() {group=Interpolate, parameterName=Lerp Interpolate Mask Type}
   B=@n23.Random Offset
   Alpha=ScalarParameter() {group=SplineThicken, parameterName=UseDynamicParameterForShapeType}
  )
  n25 = Multiply(
   A=@n14.Dissolve Edge Intensity
   B=Power(
    Base=LinearInterpolate(A=@n20, B=@n22, Alpha=@n24)
    Exp=ScalarParameter() {defaultValue=1.7, group=Interpolate, parameterName=Interpolate Power}
   )
  )
  Multiply(
   A=StaticSwitchParameter(True=Multiply(A=@n9, B=@n13), False=Constant()) {defaultValue=true, group=SplineThicken, parameterName=Use SplineThicken, sortPriority=0}
   B=StaticSwitchParameter(True=LinearInterpolate(B=@n18, Alpha=@n25), False=Constant()) {defaultValue=true, group=Interpolate, parameterName=Use Interpolate WPO, sortPriority=0}
  )
  n26 = MaterialFunctionCall("FadeSensitivity (optional)"=ScalarParameter() {defaultValue=3, parameterName=AngleFade}) {materialFunction=&FadeBasedOnViewAngle}
  n27 = TextureCoordinate() {uTiling=0.34}
  n28 = AppendVector(A=ScalarParameter() {defaultValue=1, parameterName=UMultiplier2}, B=ScalarParameter() {defaultValue=1, parameterName=VMultiplier2})
  n29 = Multiply(A=Time(), B=ScalarParameter() {defaultValue=1, parameterName=MaskSpeed})
  n30 = ScalarParameter() {defaultValue=0.12, parameterName=SpeedX1}
  n31 = ScalarParameter() {defaultValue=213, parameterName=SpeedY1}
  n32 = TextureSample(UVs=Panner(Coordinate=Multiply(A=@n27, B=@n28), Time=@n29, Speed=AppendVector(A=@n30, B=@n31)) {speedX=0.012, speedY=0.13}) {samplerType=SAMPLERTYPE_LinearColor, texture=&T_Turbulence_Seamless}
  n33 = TextureCoordinate() {uTiling=0.6, vTiling=1.5}
  n34 = AppendVector(A=ScalarParameter() {defaultValue=1, parameterName=UMultiplier3}, B=ScalarParameter() {defaultValue=1, parameterName=VMultiplier3})
  n35 = ScalarParameter() {defaultValue=0.15, parameterName=SpeedX2}
  n36 = ScalarParameter() {defaultValue=200, parameterName=SpeedY2}
  n37 = TextureSample(UVs=Panner(Coordinate=Multiply(A=@n33, B=@n34), Time=@n29, Speed=AppendVector(A=@n35, B=@n36)) {speedX=0.012, speedY=0.06}) {samplerType=SAMPLERTYPE_LinearColor, texture=&T_Turbulence_Seamless}
  n38 = ComponentMask(None=TextureCoordinate()) {g=true}
  n39 = ComponentMask(None=TextureCoordinate()) {r=true}
  n40 = Multiply(A=Multiply(A=OneMinus(None=@n38), B=@n38), B=Multiply(A=OneMinus(None=@n39), B=@n39))
  n41 = ScalarParameter() {defaultValue=0.8, parameterName=EXP}
  n42 = Multiply(
   A=Power(Base=Multiply(A=@n32.G, B=@n37.G), Exp=ScalarParameter() {defaultValue=1, parameterName=MaskPow})
   B=Multiply(A=Power(Base=@n40, Exp=@n41), B=ScalarParameter() {defaultValue=2, parameterName=ScaleAlphaMask})
  )
  Multiply(A=@n6.A, B=DepthFade(Opacity=Multiply(A=@n26.Result, B=@n42), FadeDistance=ScalarParameter() {defaultValue=10, parameterName=DepthFade}))
  ScalarParameter() {defaultValue=5, group=SplineThicken, parameterName=SplineWidthBase, sortPriority=2}
  ScalarParameter() {defaultValue=1, group=SplineThicken, parameterName=SplineWidthTip, sortPriority=1}
 ]
}
